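; Passive player augment ability that gives the occupied vehicle unlimited fuel and applies a health effect to it.
; It applies on activation if already in a vehicle, and gameplay event waits handle vehicle entry and exit.
; On authority it zeroes idle and driving fuel use, optionally boost use, and gives empty tanks minimal fuel.
; On exit it restores fuel usage unless a remaining non-driver passenger has the augment tag.
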(event Ability|EventActivateAbility
  (Utilities|IsinVehicle (Ability|GetActivatingPawn)
    (:true
      (CallFunction|ApplyZeroFuel))
    (:false
      _))
  (Ability|Tasks|WaitGameplayEvent "(TagName=\"Event.Vehicle.Entered\")"
    (:then
      (Ability|Tasks|WaitGameplayEvent "(TagName=\"Event.Vehicle.Exit\")"
        (:EventReceived
          (CallFunction|RemoveZeroFuel))))
    (:EventReceived
      (CallFunction|ApplyZeroFuel))))

(event Ability|EventFailedToActivatePassiveAbility
  (params (ActorInfo :type Gameplay%20Ability%20Actor%20Info%20Structure)))

(event Custom|ApplyZeroFuel
  (bind _returnvalue (CallFunction|GetAvatarasFortPlayerPawn))
  ;[0.03998 0.2729 1] Add / Remove unlimited fuel to vehicle
  (bind _output_get (Variables|Default|SetAppliedVehicle (Vehicle|GetVehicle _returnvalue)))
  (Ability|HasAuthority
    (:True
      (Class|FuelComponent|SetFuelUsageInfo (Vehicle|Fuel|GetVehicleFuelComponent _output_get) (Utilities|Struct|MakeRuntimeFuelUsageInfo 0.0 0.0 (CallFunction|GetFuelSpentBoosting)))
      (bind _returnvalue_2 (GameplayEffects|FortApplyGameplayEffectSpecToTarget :self (Ability|GetFortAbilitySystemComponent _output_get) :GameplayEffectSpecHandle (Ability|MakeOutgoingGameplayEffectSpec "/MusterPlayerAugments/Items/Augments/FuelEfficient/GE_PA_FuelEff_Health.GE_PA_FuelEff_Health_C") :TargetOfEffect (Ability|GetFortAbilitySystemComponent _output_get) :OptionalSourceObject _returnvalue)) ; Apply Healing GE to Vehicle
      (Variables|Default|SetHealthGESpec _returnvalue_2)
      (CallFunction|CheckforZeroFuel))))

(event Custom|RemoveZeroFuel
  ;[0.03998 0.2729 1] Add / Remove unlimited fuel to vehicle
  (bind _appliedvehicle (Variables|Default|GetAppliedVehicle)
    (:then
      (Ability|HasAuthority
        (:True
          (if (not (CallFunction|OccupantsSupercharged?))
            (Class|FuelComponent|ReinitializeFuelUsageInfo (Vehicle|Fuel|GetVehicleFuelComponent _appliedvehicle))))))
    (:else)))

(fn CheckForZeroFuel
  ;[0.634 0.1685 1] Set Fuel to 1 if empty to ensure player can drive it
  (bind _appliedvehicle (Variables|Default|GetAppliedVehicle)
    (:then
      (if (<= (Vehicle|Fuel|GetFuel _appliedvehicle) 0)
        (Vehicle|Fuel|SetFuel _appliedvehicle 1.0)))
    (:else)))

(fn GetFuelSpentBoosting
  (bind (_returnvalue_fuelpersecondidle _returnvalue_fuelperseconddriving _returnvalue_fuelpersecondboosting) (Utilities|Struct|BreakRuntimeFuelUsageInfo (Class|FuelComponent|GetFuelUsageInfo (Vehicle|Fuel|GetVehicleFuelComponent (Variables|Default|GetAppliedVehicle)))))
  ;[0.634 0.1685 1] Fix to disable free boosting via HF
  (return (Math|Float|SelectFloat 0.0 _returnvalue_fuelpersecondboosting (ScalableFloat|GetValueasBool (Variables|Default|GetFreeBoostingEnabled?)))))

(fn OccupantsSupercharged?
  ;[0.634 0.1685 1] JIRA fix to ensure Supercharged effect is not lost when multiple passengers have augment
  (bind _array_element (Utilities|Array|ForEachLoopwithBreak (Vehicle|GetAllNonDriverPassengers (Variables|Default|GetAppliedVehicle) :!class /Script/FortniteGame.FortAthenaVehicle))
    (:LoopBody
      (if (GameplayTags|HasMatchingGameplayTag _array_element (Variables|Default|GetAugmentTag))
        (Utilities|FlowControl|Sequence
          (:then_0
            _)
          (:then_1
            _))
        (LocalVariables|SetLocalBoolean true)))
    (:Completed
      _))
  (return (LocalVariables|GetLocalBoolean)))
